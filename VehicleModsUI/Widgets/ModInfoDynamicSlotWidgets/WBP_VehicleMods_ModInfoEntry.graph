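# Vehicle mod UI entry widget that shows one mod slot's icon with optional name, type label and description.
# SetModInfo takes a FortVehicleModConfig and fills icon and text, or calls SetEmptySlot when invalid, which hides text, shows placeholder art and can collapse the entry.
# SetEntryType maps the Mod Type gameplay tag to display text, shown per DisplayModType.
# PreConstruct previews text visibility at design time.
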WidgetBlueprint /VehicleModsUI/Widgets/ModInfoDynamicSlotWidgets/WBP_VehicleMods_ModInfoEntry
graphs [
 EventGraph {nodes [
   n1 = Event Construct()
   SetEntryType(execute=[@n3.then, @n1.then])
   n2 = Event PreConstruct()
   > Branch(Condition=@n2.IsDesignTime)
   > n3 = SetVisibility(
    self=Get VerticalBox_ModTextInfo() -> (VerticalBox_ModTextInfo:Vertical Box Object Reference)
    InVisibility=Select(Option 0=Collapsed, Option 1=Visible, Index=Get DisplayTextInfo() -> (DisplayTextInfo:bool))
   )
  ]}
 ResetEmpty {nodes [
   ResetEmpty() {is=title K2Node_FunctionEntry}
   > Set VehicleMod(VehicleMod:Fort Vehicle Mod Config Object Reference) -> (Output_Get:Fort Vehicle Mod Config Object Reference)
   > Set IsSlotEmpty(IsSlotEmpty:bool=true) -> (Output_Get:bool)
  ]}
 SetEmptySlot {nodes [
   n1 = SetEmptySlot!() -> (then, CollapseEmptySlot:bool=~) {is=title K2Node_FunctionEntry}
   > SetVisibility(InVisibility=Select(Option 0=SelfHitTestInvisible, Option 1=Collapsed, Index=@n1))
   > SetModTextInfo!(execute, self, ShouldDisplayTextInfo:bool=false) -> (then)
   > SetBrushFromMaterial(
    self=Get Image_ModIcon() -> (Image_ModIcon:Image Object Reference)
    Material=/Game/Athena/HUD/Inventory/Art/M_UI_Inventory_ResourceItem_EmptySlot
   )
   > Set IsSlotEmpty(IsSlotEmpty=true)
  ]}
 SetEntryType {nodes [
   SetEntryType() {is=title K2Node_FunctionEntry}
   > SetVisibility(
    self=Get Text_StatName() -> (Text_StatName:Common Text Object Reference)
    InVisibility=Select(Option 0=Collapsed, Option 1=SelfHitTestInvisible, Index=Get DisplayModType() -> (DisplayModType:bool))
   )
   n1 = Find(
    TargetMap:Map of Gameplay Tag Structures to Texts=Get ModType_DisplayNames() -> (ModType_DisplayNames:Map of Gameplay Tag Structures to Texts)
    Key:"Gameplay Tag Structure (by ref)"=Get Mod Type() -> (Mod Type:Gameplay Tag Structure)
   ) -> (Value:Text)
   > Branch(Condition=@n1.ReturnValue)
   > "SetText (Text)"(self=Get Text_StatName(), InText=@n1)
  ]}
 SetModIcon {nodes [
   SetModIcon() {is=title K2Node_FunctionEntry}
   > SetDesiredSizeOverride!(
    execute
    self:Image Object Reference=Get Image_ModIcon()
    DesiredSize:Vector 2D Structure=Get DesiredIconSize() -> (DesiredIconSize:Vector 2D Structure)
   ) -> (then)
   n1 = Get SmallIconImage!(self:Fort Vehicle Mod Config Object Reference=Get VehicleMod() -> (VehicleMod:Fort Vehicle Mod Config Object Reference)) -> (SmallIconImage:Texture 2D Soft Object Reference)
   > n2 = Branch(Condition=IsValidSoftObjectReference(SoftObjectReference=@n1))
   > SetBrushFromSoftTexture!(execute, self:Image Object Reference=Get Image_ModIcon(), SoftTexture:Texture 2D Soft Object Reference=@n1, bMatchSize:bool=false) -> (then)
   SetBrushFromMaterial(execute=@n2.else, self=Get Image_ModIcon(), Material=&M_UI_Inventory_ResourceItem_EmptySlot)
  ]}
 SetModInfo {nodes [
   n1 = SetModInfo!() -> (then, InVehicleModConfig:Fort Vehicle Mod Config Object Reference) {is=title K2Node_FunctionEntry}
   > n2 = Set VehicleMod(VehicleMod=@n1)
   > n3 = Is Valid(InputObject=@n2)
   > SetModIcon() {is=title K2Node_CallFunction}
   > SetModTextInfo(ShouldDisplayTextInfo=Get DisplayTextInfo())
   > Set IsSlotEmpty(IsSlotEmpty=false)
   > SetVisibility(InVisibility=HitTestInvisible)
   SetEmptySlot!(execute=@n3.Is Not Valid, self, CollapseEmptySlot:bool=false) -> (then) {is=title K2Node_CallFunction}
  ]}
 SetModTextInfo {nodes [
   n1 = SetModTextInfo!() -> (then, ShouldDisplayTextInfo:bool=~) {is=title K2Node_FunctionEntry}
   > n2 = Branch(Condition=@n1)
   > "SetText (Text)"(
    self=Get Text_ModName() -> (Text_ModName:Common Text Object Reference)
    InText=Get DisplayName!(self:Fort Vehicle Mod Config Object Reference=Get VehicleMod()) -> (DisplayName:Text)
   )
   n3 = Get Description!(self:Fort Vehicle Mod Config Object Reference=Get VehicleMod()) -> (Description:Text)
   n4 = Get Text_ModDescription() -> (Text_ModDescription:Common Text Object Reference)
   > "SetText (Text)"(self=@n4, InText=@n3)
   > SetVisibility(self=@n4, InVisibility=Select(Option 0=SelfHitTestInvisible, Option 1=Collapsed, Index=TextIsEmpty(InText=@n3)))
   > SetEntryType() {is=title K2Node_CallFunction}
   > SetVisibility(self=Get VerticalBox_ModTextInfo(), InVisibility=SelfHitTestInvisible)
   SetVisibility(execute=@n2.else, self=Get VerticalBox_ModTextInfo(), InVisibility=Collapsed)
  ]}
]
variables [
 DesiredIconSize {type=Vector2D, default=<32 32>}
 DisplayModType {type=boolean}
 DisplayTextInfo {type=boolean, default=true}
 IsSlotEmpty {type=boolean, default=true}
 Mod Type {type=GameplayTag, default {tagName=None}}
 VehicleMod {type=FortVehicleModConfig, default=null}
]
widgets {
 info {inheritedWidgetCount=0, parentClass=&UserWidget, rootWidgetClass=&Overlay, widgetCount=7}
 widgets [
  Overlay_Root {widgetClassPath=&Overlay, widget=Overlay_Root}
  HorizontalBox_24 {
   widgetClassPath=&HorizontalBox
   parent=Overlay_Root
   widget=HorizontalBox_24
   properties {flowDirectionPreference=Culture}
   slot {class=OverlaySlot, properties {horizontalAlignment=HAlign_Fill, verticalAlignment=VAlign_Fill, padding=<ltrb 2.5 2.5 2.5 2.5>}}
  }
  Image_ModIcon {
   widgetClassPath=&Image
   bIsVariable=true
   parent=HorizontalBox_24
   widget=Image_ModIcon
   properties {
    visibility=SelfHitTestInvisible
    brush {
     bIsDynamicallyLoaded=false
     drawAs=Box
     imageType=NoImage
     mirroring=NoMirror
     resourceName=None
     resourceObject=&M_UI_Inventory_ResourceItem_EmptySlot
     tiling=NoTile
     imageSize=<32 32>
     margin=<ltrb 0.5 0.5 0.5 0.5>
     outlineSettings {
      bUseBrushTransparency=false
      roundingType=HalfHeightRadius
      width=0
      color {colorUseRule=UseColor_Specified, specifiedColor=<rgba 0 0 0 0>}
      cornerRadii=<0 0 0 0>
     }
     tintColor {colorUseRule=UseColor_Specified, specifiedColor=<rgba 1 1 1 1>}
     uVRegion {bIsValid=false, max=<0 0>, min=<0 0>}
    }
   }
   slot {class=HorizontalBoxSlot, properties {horizontalAlignment=HAlign_Center, verticalAlignment=VAlign_Center}}
  }
  VerticalBox_ModTextInfo {
   widgetClassPath=&VerticalBox
   bIsVariable=true
   parent=HorizontalBox_24
   widget=VerticalBox_ModTextInfo
   slot {class=HorizontalBoxSlot, properties {padding=<ltrb 10 0 0 0>, size {sizeRule=Fill, value=1}}}
  }
  Text_ModName {
   widgetClassPath=&CommonTextBlock
   bIsVariable=true
   parent=VerticalBox_ModTextInfo
   widget=Text_ModName
   properties {
    flowDirectionPreference=Culture
    lineHeightPercentage=0.9
    mobileFontSizeMultiplier=1.2
    scrollStyle=&TextScrollStyle-Base_C
    text="{Mod Name}"
    textTransformPolicy=ToUpper
    font {
     bForceMonospaced=false
     bMaterialIsStencil=false
     fontMaterial=/Game/Athena/HUD/Inventory/Art/M_UI_Inventory_CatergoryHeader_FontColor
     fontObject=&BurbankSmall
     letterSpacing=0
     monospacedWidth=1
     size=13.5
     skewAmount=0.25
     typefaceFontName=Black
     outlineSettings {
      bApplyOutlineToDropShadows=false
      bMiteredCorners=false
      bSeparateFillAlpha=true
      outlineMaterial=&M-FontShadow
      outlineSize=2
      outlineColor=<rgba 0 0 0 0.5>
     }
    }
    shadowOffset=<0 0>
   }
   slot {class=VerticalBoxSlot}
  }
  Text_StatName {
   widgetClassPath=&CommonTextBlock
   bIsVariable=true
   parent=VerticalBox_ModTextInfo
   widget=Text_StatName
   properties {
    flowDirectionPreference=Culture
    mobileFontSizeMultiplier=1.15
    scrollStyle=/FortUILibrary/Styles/Typography/Foundation/TextScrollStyle-NoFade.TextScrollStyle-NoFade_C
    text="{Stat Name}"
    textTransformPolicy=ToUpper
    font {
     bForceMonospaced=false
     bMaterialIsStencil=false
     fontMaterial=null
     fontObject=&BurbankSmall
     letterSpacing=0
     monospacedWidth=1
     size=10.5
     skewAmount=0
     typefaceFontName=Black
     outlineSettings {
      bApplyOutlineToDropShadows=false
      bMiteredCorners=false
      bSeparateFillAlpha=false
      outlineMaterial=null
      outlineSize=0
      outlineColor=<rgba 0 0 0 1>
     }
    }
    shadowOffset=<0 0>
   }
   slot {class=VerticalBoxSlot}
  }
  Text_ModDescription {
   widgetClassPath=&CommonTextBlock
   bIsVariable=true
   parent=VerticalBox_ModTextInfo
   widget=Text_ModDescription
   properties {
    flowDirectionPreference=Culture
    mobileFontSizeMultiplier=1.35
    text="{Placeholder description text. Item Description would go here.}"
    wrapTextAt=250
    font {
     bForceMonospaced=false
     bMaterialIsStencil=false
     fontMaterial=/Game/Athena/HUD/Inventory/Art/M_UI_Inventory_ItemInfo_Description_FontColor
     fontObject=&BurbankSmall
     letterSpacing=0
     monospacedWidth=1
     size=10.5
     skewAmount=0
     typefaceFontName=Bold
     outlineSettings {
      bApplyOutlineToDropShadows=false
      bMiteredCorners=false
      bSeparateFillAlpha=false
      outlineMaterial=null
      outlineSize=0
      outlineColor=<rgba 0 0 0 1>
     }
    }
    shadowOffset=<0 0>
   }
   slot {class=VerticalBoxSlot}
  }
 ]
}
