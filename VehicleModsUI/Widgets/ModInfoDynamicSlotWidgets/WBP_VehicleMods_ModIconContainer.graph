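# Marker details extension widget that shows a vehicle's installed mods as a row of icon-only entries for front bumper, roof and tire slots.
# Construct maps each slot's gameplay tag to its ModInfoEntry widget.
# SetModInfo looks up the config's slot tag, passes the config to that entry and shows the container.
# ResetModInfo empties and collapses every entry, then collapses the container, while SetIsModifiable toggles container visibility.
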WidgetBlueprint /VehicleModsUI/Widgets/ModInfoDynamicSlotWidgets/WBP_VehicleMods_ModIconContainer
graphs [EventGraph {nodes [
   Event Construct()
   > Set ModSlot_Widgets(ModSlot_Widgets:Map of Gameplay Tag Structures to WBP Vehicle Mods Mod Info Entry Object References=Make Map!(
     Key 0:Gameplay Tag Structure=(TagName="Vehicle.Mod.Slot.FrontBumper")
     Value 0:WBP Vehicle Mods Mod Info Entry Object Reference=Get ModSlot_FrontBumper() -> (ModSlot_FrontBumper:WBP Vehicle Mods Mod Info Entry Object Reference)
     Key 1:Gameplay Tag Structure=(TagName="Vehicle.Mod.Slot.Roof")
     Value 1:WBP Vehicle Mods Mod Info Entry Object Reference=Get ModSlot_Roof() -> (ModSlot_Roof:WBP Vehicle Mods Mod Info Entry Object Reference)
     Key 2:Gameplay Tag Structure=(TagName="Vehicle.Mod.Slot.Tire")
     Value 2:WBP Vehicle Mods Mod Info Entry Object Reference=Get ModSlot_Tires() -> (ModSlot_Tires:WBP Vehicle Mods Mod Info Entry Object Reference)
    ) -> (Map:Map of Gameplay Tag Structures to WBP Vehicle Mods Mod Info Entry Object References) {is=title K2Node_MakeMap}) -> (Output_Get:Map of Gameplay Tag Structures to WBP Vehicle Mods Mod Info Entry Object References)
   K2Node_Event_3 = SetIsModifiable!() -> (OutputDelegate, then, bIsModifiable:bool=~) {is=title K2Node_CustomEvent}
   > SetVisibility(InVisibility=Select(Option 0=Collapsed, Option 1=HitTestInvisible, Index=@K2Node_Event_3.bIsModifiable))
   n1 = Event SetModInfo!() -> (OutputDelegate, then, VehicleModConfig:Fort Vehicle Mod Config Object Reference) {is=title K2Node_Event}
   n2 = Find(
    TargetMap:Map of Gameplay Tag Structures to WBP Vehicle Mods Mod Info Entry Object References=Get ModSlot_Widgets() -> (ModSlot_Widgets:Map of Gameplay Tag Structures to WBP Vehicle Mods Mod Info Entry Object References)
    Key:"Gameplay Tag Structure (by ref)"=Get ModSlot!(self:Fort Vehicle Mod Config Object Reference=@n1.VehicleModConfig) -> (ModSlot:Gameplay Tag Structure)
   ) -> (Value:WBP Vehicle Mods Mod Info Entry Object Reference)
   > Branch(Condition=@n2.ReturnValue)
   > SetModInfo!(
    execute
    self:WBP Vehicle Mods Mod Info Entry Object Reference=@n2
    InVehicleModConfig:Fort Vehicle Mod Config Object Reference=@n1.VehicleModConfig
   ) -> (then)
   > SetVisibility(InVisibility=HitTestInvisible)
   Event ResetModInfo() {is=title K2Node_Event}
   > n3 = Values!(execute, TargetMap:Map of Gameplay Tag Structures to WBP Vehicle Mods Mod Info Entry Object References=Get ModSlot_Widgets()) -> (then, Values:Array of WBP Vehicle Mods Mod Info Entry Object References)
   > n4 = For Each Loop(Array:Array of WBP Vehicle Mods Mod Info Entry Object References=@n3) -> (Array Element:WBP Vehicle Mods Mod Info Entry Object Reference)
   SetVisibility(execute=@n4.Completed, InVisibility=Collapsed)
   SetEmptySlot!(execute=@n4.LoopBody, self:WBP Vehicle Mods Mod Info Entry Object Reference=@n4, CollapseEmptySlot:bool=true) -> (then)
  ]}]
variables [ModSlot_Widgets {default {}}]
widgets {
 info {
  inheritedWidgetCount=0
  parentClass=/Script/VehicleModsUI.AthenaMarkerDetailsExtensionVehicleMods
  rootWidgetClass=&Overlay
  widgetCount=5
 }
 widgets [
  Overlay_Root {widgetClassPath=&Overlay, widget=Overlay_Root}
  HorizontalBox_40 {widgetClassPath=&HorizontalBox, parent=Overlay_Root, widget=HorizontalBox_40, slot {class=OverlaySlot}}
  ModSlot_FrontBumper {
   widgetClassPath=&/WBP_VehicleMods_ModInfoEntry
   bIsVariable=true
   parent=HorizontalBox_40
   widget=ModSlot_FrontBumper
   properties {displayTextInfo=false, mod Type {tagName=Vehicle.Mod.Slot.FrontBumper}}
   slot {class=HorizontalBoxSlot}
  }
  ModSlot_Roof {
   widgetClassPath=&WBP_VehicleMods_ModInfoEntry
   bIsVariable=true
   parent=HorizontalBox_40
   widget=ModSlot_Roof
   properties {displayTextInfo=false, mod Type {tagName=Vehicle.Mod.Slot.Roof}}
   slot {class=HorizontalBoxSlot}
  }
  ModSlot_Tires {
   widgetClassPath=&WBP_VehicleMods_ModInfoEntry
   bIsVariable=true
   parent=HorizontalBox_40
   widget=ModSlot_Tires
   properties {displayTextInfo=false, mod Type {tagName=Vehicle.Mod.Slot.Tire}}
   slot {class=HorizontalBoxSlot}
  }
 ]
}
