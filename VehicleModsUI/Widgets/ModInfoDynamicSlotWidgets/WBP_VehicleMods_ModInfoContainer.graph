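# Vehicle mods interaction-info widget that shows mod details in three stacked entries for front bumper, roof and tire slots.
# Construct builds a map from slot gameplay tags to entry widgets, then sets every entry to a collapsed empty state, as ResetModInfo also does.
# Valid context-info updates go to HandleContextInfoUpdated.
# SetModInfo looks up the config's slot tag in the map and passes the config to the matching entry.
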WidgetBlueprint /VehicleModsUI/Widgets/ModInfoDynamicSlotWidgets/WBP_VehicleMods_ModInfoContainer
graphs [EventGraph {nodes [
   Event Construct()
   > n1 = Set ModSlot_Widgets(ModSlot_Widgets:Map of Gameplay Tag Structures to WBP Vehicle Mods Mod Info Entry Object References=Make Map!(
     Key 0:Gameplay Tag Structure=(TagName="Vehicle.Mod.Slot.FrontBumper")
     Value 0:WBP Vehicle Mods Mod Info Entry Object Reference=Get ModSlot_FrontBumper() -> (ModSlot_FrontBumper:WBP Vehicle Mods Mod Info Entry Object Reference)
     Key 1:Gameplay Tag Structure=(TagName="Vehicle.Mod.Slot.Roof")
     Value 1:WBP Vehicle Mods Mod Info Entry Object Reference=Get ModSlot_Roof() -> (ModSlot_Roof:WBP Vehicle Mods Mod Info Entry Object Reference)
     Key 2:Gameplay Tag Structure=(TagName="Vehicle.Mod.Slot.Tire")
     Value 2:WBP Vehicle Mods Mod Info Entry Object Reference=Get ModSlot_Tires() -> (ModSlot_Tires:WBP Vehicle Mods Mod Info Entry Object Reference)
    ) -> (Map:Map of Gameplay Tag Structures to WBP Vehicle Mods Mod Info Entry Object References) {is=title K2Node_MakeMap}) -> (Output_Get:Map of Gameplay Tag Structures to WBP Vehicle Mods Mod Info Entry Object References)
   n2 = Values!(
    execute=[@n5.then, @n1.then]
    TargetMap:Map of Gameplay Tag Structures to WBP Vehicle Mods Mod Info Entry Object References=Get ModSlot_Widgets() -> (ModSlot_Widgets:Map of Gameplay Tag Structures to WBP Vehicle Mods Mod Info Entry Object References)
   ) -> (then, Values:Array of WBP Vehicle Mods Mod Info Entry Object References)
   > n3 = For Each Loop(Array:Array of WBP Vehicle Mods Mod Info Entry Object References=@n2) -> (Array Element:WBP Vehicle Mods Mod Info Entry Object Reference)
   > SetEmptySlot!(execute, self:WBP Vehicle Mods Mod Info Entry Object Reference=@n3, CollapseEmptySlot:bool=true) -> (then)
   n4 = Event OnContextInfoUpdated!() -> (OutputDelegate, then, ContextInfo:Fort Interact Context Info Object Reference) {is=title K2Node_Event}
   > Is Valid(InputObject=@n4.ContextInfo)
   > HandleContextInfoUpdated!(
    execute
    self:Fort Interaction Context Info Widget Vehicle Mods Object Reference
    ContextInfo:Fort Interact Context Info Object Reference=@n4.ContextInfo
   ) -> (then)
   n5 = Event ResetModInfo() {is=title K2Node_Event}
   n6 = Event SetModInfo!() -> (OutputDelegate, then, VehicleModConfig:Fort Vehicle Mod Config Object Reference) {is=title K2Node_Event}
   n7 = Find(
    TargetMap:Map of Gameplay Tag Structures to WBP Vehicle Mods Mod Info Entry Object References=Get ModSlot_Widgets()
    Key:"Gameplay Tag Structure (by ref)"=Get ModSlot!(self:Fort Vehicle Mod Config Object Reference=@n6.VehicleModConfig) -> (ModSlot:Gameplay Tag Structure)
   ) -> (Value:WBP Vehicle Mods Mod Info Entry Object Reference)
   > Branch(Condition=@n7.ReturnValue)
   > SetModInfo!(
    execute
    self:WBP Vehicle Mods Mod Info Entry Object Reference=@n7
    InVehicleModConfig:Fort Vehicle Mod Config Object Reference=@n6.VehicleModConfig
   ) -> (then)
  ]}]
variables [
 Cached Receiving Actor {type=Actor, default=null}
 ModSlot_Widgets {default {}}
]
widgets {
 info {
  inheritedWidgetCount=0
  parentClass=/Script/VehicleModsUI.FortInteractionContextInfoWidgetVehicleMods
  rootWidgetClass=&VerticalBox
  widgetCount=4
 }
 widgets [
  VerticalBox_23 {widgetClassPath=&VerticalBox, widget=VerticalBox_23}
  ModSlot_FrontBumper {
   widgetClassPath=&/WBP_VehicleMods_ModInfoEntry
   bIsVariable=true
   parent=VerticalBox_23
   widget=ModSlot_FrontBumper
   properties {mod Type {tagName=Vehicle.Mod.Slot.FrontBumper}}
   slot {class=VerticalBoxSlot}
  }
  ModSlot_Roof {
   widgetClassPath=&WBP_VehicleMods_ModInfoEntry
   bIsVariable=true
   parent=VerticalBox_23
   widget=ModSlot_Roof
   properties {mod Type {tagName=Vehicle.Mod.Slot.Roof}}
   slot {class=VerticalBoxSlot}
  }
  ModSlot_Tires {
   widgetClassPath=&WBP_VehicleMods_ModInfoEntry
   bIsVariable=true
   parent=VerticalBox_23
   widget=ModSlot_Tires
   properties {mod Type {tagName=Vehicle.Mod.Slot.Tire}}
   slot {class=VerticalBoxSlot}
  }
 ]
}
